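; Soy Boarding gameplay ability that launches the player skyward, then redeploys a glider with a cosmetic override.
; Activation applies gameplay effects and calls Leap; a montage notify later force-opens the glider after a client-wait net sync.
; Landing, movement-mode changes, rift skydiving and vent entry trigger cleanup, which ends skydiving, restores HUD and removes an effect to start cooldown.
; Ending before glider deployment also starts cooldown; Surf retains older float logic.
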(event Ability|EventActivateAbility
  ;[1 1 1] ActivateAbility->Send us into the sky, and then deploy the soy glider
  (|CallParentFunction)
  (Utilities|PowerCommitAbility (Variables|AbilityCommit|GetIgnoreAbilityCommit)
    (:Pass
      _)
    (:Fail
      (Development|PrintString "Power: Ability Commit Failed")
      (CallFunction|PowerCommitFailed))
    (:Ignored
      _))
  (Utilities|FlowControl|Sequence
    (:then_0
      ;[1 1 1] Shameful mess is failsafe states - situations we'd encounter during gameplay that we'd need to clean the ability up for
      (Ability|Tasks|WaitGameplayTagAdd "(TagName=\"Gameplay.Status.Rift.Skydiving\")"
        (:Added
          (Ability|Tasks|WaitNetSync "OnlyServerWait"
            (:OnSync
              _)) ; Server gets here first with states applied via GE
        )))
    (:then_1
      (Ability|Tasks|WaitGameplayEvent "(TagName=\"Abilities.Generic.Vent.Enter\")"
        (:EventReceived
          (Ability|Tasks|WaitNetSync "OnlyClientWait"
            (:OnSync
              _)) ; Client gets here first with stuff that launches via event
        )))
    (:then_2
      (CallFunction|Target) ; Break da buildin's
      (HUD|SetHUDElementVisibility (BlueprintContext|GetFortHUDContext) (Variables|Default|GetT_SkyDivingMessageTag) true "None") ; Hide skydiving message
      (CallFunction|Leap)
      (Ability|ApplyGameplayEffectToOwner "/HighTower/Items/Soy/GE_HighTower_Soy_Boarding_MinTime.GE_HighTower_Soy_Boarding_MinTime_C")
      (Ability|ApplyGameplayEffectToOwner "/HighTower/Items/Soy/GE_HighTower_Soy_Boarding_Active_Tag.GE_HighTower_Soy_Boarding_Active_Tag_C") ; This is a tag for the cancel ability to check so that we dont just end any skydiving ability :D
    ))
  (Variables|Default|SetGenericFailsafeHit true) ; Kill gliding if we hit a failsafe condition
  (CallFunction|GenericFailSafe))

(event EventOnLand
  (params (Hit :type Hit%20Result%20Structure%20%28by%20ref%29))
  ;[1 1 1] Clean up leap feedback and ability if we land, our movement mode changes while we're gliding, or we hit a Rift.
  (|CallParentFunction Hit)
  (CallFunction|CleanupAbility))

(event Custom|MovementModeChangeFailSafe
  (params (Character :type Character%20Object%20Reference) (PrevMovementMode :type EMovementMode%20Enum :default "MOVE_None") (PreviousCustomMode :type Byte))
  ;[1 1 1] Clean up leap feedback and ability if we land, our movement mode changes while we're gliding, or we hit a Rift.
  (Character|UnbindEventfromMovementModeChangedDelegate Character (delegate Custom|MovementModeChangeFailSafe))
  (|CallParentFunction (stale EventOnLand Hit))
  (CallFunction|CleanupAbility))

(event Custom|GenericFailSafe
  ;[1 1 1] Clean up leap feedback and ability if we land, our movement mode changes while we're gliding, or we hit a Rift.
  (|CallParentFunction (stale EventOnLand Hit))
  (CallFunction|CleanupAbility))

(event Custom|CleanupAbility
  ;[1 1 1] Ability cleanup - start cooldown, remove FX, return skydiving prompt to hude and a failsafe to end skydiving
  (Actor|DestroyActor (Variables|Default|GetLensEffect))
  (Parachute|EndSkydiving (Variables|Default|GetPlayerPawn)) ; Failsafe
  (Ability|RemoveGameplayEffectFromOwnerWithAssetTags "(GameplayTags=((TagName=\"Asset.Cooldown.HighTower.Soy.Boarding\")))") ; Removing this startst he actual cooldown
  (if (GameplayTags|HasMatchingGameplayTag (Variables|Default|GetPlayerPawn) "(TagName=\"Granted.HighTower.Power.Soy.Boarding.Active\")")
    (Ability|RemoveGameplayEffectFromOwnerWithGrantedTags "(GameplayTags=((TagName=\"Granted.HighTower.Power.Soy.Boarding.Active\")))"))
  (HUD|SetHUDElementVisibility (BlueprintContext|GetFortHUDContext) (Variables|Default|GetT_SkyDivingMessageTag) false "None")
  (Ability|EndAbility))

(event EventCastMontageBlendOut)

(event EventCastMontageNotifyEndAbility)

(event EventCastMontageNotifyStart
  (params (NotifyTag :type Gameplay%20Tag%20Structure))
  (bind _playerpawn (Variables|Default|GetPlayerPawn))
  ;[1 1 1] We play a montage to fuse the JumpPad anims with actual glider redeploy.  This happens here.
  (if (and (GameplayTags|Equal(GameplayTag) NotifyTag (Variables|Default|GetTagStartRedeploy)) (not (Variables|Default|GetGenericFailsafeHit)))
    (Ability|Tasks|WaitNetSync "OnlyClientWait"
      (:OnSync
        (Variables|Default|SetParachuteOpened true)
        (Utilities|FlowControl|Sequence
          (:then_0
            (Tags|UnBlockAbilitiesWithTags (Ability|GetFortAbilitySystemComponent (Variables|Default|GetPlayerPawn)) (Variables|Default|GetT_OpenParahcute) false))
          (:then_1
            ;; PlayGliderOpen
            (Audio|SpawnSoundAttached (Variables|Default|GetOpenGliderSoundCue) (Variables|/Script/Engine.Actor|GetRootComponent _playerpawn))
            (Parachute|ForceOpenParachuteandOverrideGlider _playerpawn (Variables|Default|GetGlider%20Cosmetic%20Override) true true)
            (Character|BindEventtoMovementModeChangedDelegate _playerpawn (EventDispatchers|CreateEvent :!event Custom|MovementModeChangeFailSafe))))))
    (else
      (|CallParentFunction NotifyTag))))

(event EventCastMontageCancelledOrInterrupted)

(event EventPowerCommitFailed)

(event Ability|EventOnEndAbility
  (params (bWasCancelled :type Boolean))
  ;[1 1 1] Event EndAbility-> Failsafe for being dropped due to overflow. Dropped due to overflow delegate doesn't seem to fire.  Basically if it's dropped before glider redeploy, make it work still
  (if (Variables|Default|GetParachuteOpened)
    (Variables|Default|SetParachuteOpened false)
    (else
      (Ability|RemoveGameplayEffectFromOwnerWithAssetTags "(GameplayTags=((TagName=\"Asset.Cooldown.HighTower.Soy.Boarding\")))") ; Removing this startst he actual cooldown
      (Ability|RemoveGameplayEffectFromOwnerWithGrantedTags "(GameplayTags=((TagName=\"Granted.HighTower.Power.Soy.Boarding.Active\")))")))
  (|CallParentFunction bWasCancelled))

(event Custom|Surf
  ;[0.3715 0 1] (most of the) Old version in case we need it
  ;;[1 1 1] LeapFinish-> Wait for the movement logic from the parent ability to finish, and then put us in "glide" mode
  (Character|StopJumping (Variables|Default|GetPlayerPawn))
  (Character|UnbindEventfromOnReachedJumpApex (Variables|Default|GetPlayerPawn) (EventDispatchers|CreateEvent :!event Custom|Surf))
  (Mantis|Ability|Tasks|ApplyRootMotionFallingBoostForce (Variables|BoardMovementVars|GetBoardTaskInstance) (Variables|BoardMovementVars|GetBoardVerticalImpulse) (ScalableFloat|GetValueatLevel (Variables|BoardMovementVars|GetData_BoardFloatDuration)) (Variables|BoardMovementVars|GetBoardGravityScalar) (Variables|BoardMovementVars|GetBoardMaxAcceleration) (Variables|BoardMovementVars|GetBoardLateralFriction) (Variables|BoardMovementVars|GetBoardMaxSpeed)
    (:then
      (Ability|Tasks|WaitGameplayEvent "(TagName=\"Event.HighTower.Power.Soy.Boarding.ForceEnd\")"
        (:EventReceived
          (Ability|EndAbility))) ; End the ability if repressed after this point
    )
    (:OnFinish
      (Ability|EndAbility))) ; floating root motion - keeps us in the air.  High manueverability
)

(event Ability|EventOnAbilityInputReleased
  ;[0.3715 0 1] (most of the) Old version in case we need it
  ;;[1 1 1] End the ability when the input is released.  Make sure ability runs for the minimum duration
  (|CallParentFunction))

(fn CheckMinSurfTime
  (return (> (- (Utilities|Time|GetGameTimeinSeconds) (Variables|Default|GetSurfStartTime)) (Variables|Default|GetMinSurfTime)) (- (Variables|Default|GetMinSurfTime) (- (Utilities|Time|GetGameTimeinSeconds) (Variables|Default|GetSurfStartTime)))))

(fn InitAbility
  (|CallParentFunction)
  (Variables|Default|SetSurfStartTime (Utilities|Time|GetGameTimeinSeconds))
  (Variables|Default|SetGenericFailsafeHit false)
  (bind _returnvalue (Feedback|SpawnGenericCameraLensEffect (Variables|Default|GetPlayerPawn) "/HighTower/Items/Soy/Proto/BP_HighTower_Soy_Boarding__Camera_LensEffect.BP_HighTower_Soy_Boarding__Camera_LensEffect_C"))
  (Variables|Default|SetLensEffect (Utilities|Casting|CastToEmitterCameraLensEffectBase _returnvalue)))

(fn ManageInputScale
  (params (ReduceSteering :type Boolean))
  (bind _returnvalue (Fortnite|GetFortPlayerControllerfromActor (Variables|Default|GetPlayerPawn)))
  (bind _playerpawn (Variables|Default|GetPlayerPawn))
  (if (FortPawn|IsDead _playerpawn)
    ;[1 1 1] Have to get the controller this way if the pawns dead.
    (bind _asfort_player_controller (Utilities|Casting|CastToFortPlayerController (Actor|GetOwner (Variables|/Script/Engine.Pawn|GetPlayerState _playerpawn))))
    (Input|ResetInputRotationScale _asfort_player_controller)
    (elif ReduceSteering
      (Input|SetInputPitchScale _returnvalue (Variables|Default|GetPitchInputScale))
      (Input|SetInputYawScale _returnvalue (Variables|Default|GetYawInputScale))
      (else
        (Input|ResetInputRotationScale _returnvalue)))))
