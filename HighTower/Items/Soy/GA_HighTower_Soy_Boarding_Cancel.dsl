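; Gameplay ability that cancels the HighTower Soy boarding power while the player pawn is skydiving.
; On activation it does nothing if the cancel cooldown tag is present; otherwise it requires the boarding active tag and skydiving.
; On success it ends skydiving, sets movement to falling, removes the boarding active effect and leaping loop cue, then ends.
; A failsafe EndAbility covers the other paths, and a comment notes trouble sending events.
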(event Ability|EventActivateAbility
  (bind _asfort_player_pawn (Utilities|Casting|CastToFortPlayerPawn (Ability|GetAvatarActorfromActorInfo))
    (:then
      (if (GameplayTags|HasMatchingGameplayTag _asfort_player_pawn "(TagName=\"Granted.Cooldown.HighTower.Soy.Boarding.Cancel\")")
        _
        (elif (and (GameplayTags|HasMatchingGameplayTag _asfort_player_pawn "(TagName=\"Granted.HighTower.Power.Soy.Boarding.Active\")") (Parachute|IsSkydiving _asfort_player_pawn))
          (CallFunction|RanAbilitySuccesfully))))
    (:CastFailed
      _))
  (Ability|EndAbility) ; Failsafe to do nothing and end the ability if pawn isn't skydiving
)

(event Custom|RanAbilitySuccesfully
  ;[1 1 1] Ending skydiving stuff in here when cancelled - having issues sending events.
  (Parachute|EndSkydiving (stale Ability|EventActivateAbility _asfort_player_pawn))
  (Pawn|Components|CharacterMovement|SetMovementMode (Variables|/Script/Engine.Character|GetCharacterMovement (stale Ability|EventActivateAbility _asfort_player_pawn)) "MOVE_Falling" 0)
  (Ability|RemoveGameplayEffectFromOwnerWithGrantedTags "(GameplayTags=((TagName=\"Granted.HighTower.Power.Soy.Boarding.Active\")))")
  (Ability|RemoveGameplayCueFromOwner (Variables|Default|GetGC_LeapingLoop))
  (Ability|EndAbility))
